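; Manager gameplay ability for SpaceClique's in-air dash (SilkyWolf), tracking dash charges on the player pawn.
; Tag-wait tasks grant charges on medallion pickup, spend one per committed dash and start expiration on medallion loss or water exit.
; Server-side changes use SetByCaller effects capped at max, refresh status effects and clean up when empty without a medallion.
; Timers drain unused and AI bot charges; the recharge timer start is an empty stub.
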(event Ability|EventFailedToActivatePassiveAbility
  (params (ActorInfo :type Gameplay%20Ability%20Actor%20Info%20Structure)))

(event Ability|EventActivateAbility
  (bind _returnvalue (CallFunction|GetAvatarasFortPlayerPawn))
  (bind _row_runtimeraftermedalliondrop (Variables|Default|GetRow_RunTimerAfterMedallionDrop))
  (bind _returnvalue_2 (ScalableFloat|GetValueasBool _row_runtimeraftermedalliondrop))
  ;[1 1 1] Ability Activate
  (Variables|Components|SetPlayer%20Pawn _returnvalue)
  (Variables|Default|SetAbility%20System%20Component (Ability|GetAbilitySystemComponent _returnvalue))
  (Variables|Default|SetPlayer%20Controller (Fortnite|GetFortPlayerControllerfromActor _returnvalue))
  (Variables|Default|SetMax_Dash_Charges (ScalableFloat|GetValueasInteger (Variables|Default|GetRow_Max_Dash_Charges)))
  (Ability|SendGameplayEvent "(TagName=\"Gameplay.Status.SunRose.SilkyWolf.Tutorial.Start\")")
  (bind _asynctaskproxy (Ability|Tasks|WaitGameplayTagAdd "(TagName=\"Granted.Athena.SunRose.SilkyWolf.DashCommitted\")")
    (:then
      (Variables|AsyncTasks|SetAsync%20Task%20Cooldown _asynctaskproxy)
      (bind _asynctaskproxy_2 (Ability|Tasks|WaitGameplayTagAdd (Variables|Default|GetTag_SilkyWolf_Medallion))
        (:then
          (Variables|AsyncTasks|SetAsync%20Task%20Medallion _asynctaskproxy_2)
          (Utilities|FlowControl|Sequence
            (:then_0
              (bind _asynctaskproxy_3 (Ability|Tasks|WaitGameplayTagAdd (Variables|Default|GetTag_SilkyWolf_InWater))
                (:then
                  (Variables|AsyncTasks|SetAsync%20Task%20In%20Water _asynctaskproxy_3)
                  (CallFunction|StartAIBotDashRemoval))
                (:Added
                  _)))
            (:then_1
              _))
          (CallFunction|GrantInitialCharges)
          (bind _asynctaskproxy_4 (Ability|Tasks|WaitGameplayTagRemove (Variables|Default|GetTag_SilkyWolf_InWater))
            (:then
              (Variables|Default|SetAsync%20Task%20Exit%20Water _asynctaskproxy_4))
            (:Removed
              (CallFunction|StartNoUsageExpirationTimer))))
        (:Added
          (CallFunction|GrantInitialCharges)
          (Utilities|IsValid (Variables|AsyncTasks|GetAsync%20Task%20Remove%20Medallion)
            (:Is%20Not%20Valid
              (bind _asynctaskproxy_5 (Ability|Tasks|WaitGameplayTagRemove (Variables|Default|GetTag_SilkyWolf_Medallion))
                (:then
                  (Variables|AsyncTasks|SetAsync%20Task%20Remove%20Medallion _asynctaskproxy_5))
                (:Removed
                  (if (not _returnvalue_2)
                    (Utilities|Time|ClearandInvalidateTimerbyHandle (Variables|Default|GetTimer%20Handle%20Recharge%20Time)))
                  (bind _remaining_charges (CallFunction|GetRemainingCharges))
                  (if (> _remaining_charges 0)
                    (CallFunction|StartNoUsageExpirationTimer)
                    (elif _returnvalue_2
                      (Ability|RemoveGameplayEffectFromOwnerWithGrantedTags "(GameplayTags=((TagName=\"Granted.Athena.SunRose.SilkyWolf.DashCharge\")))") ; Removes medallion FX
                      (else
                        (CallFunction|Cleanup)))))))) ; Could alerady be valid if picked up after recently dropped
        )))
    (:Added
      (CallFunction|UseDashCharge -1))))

(event Custom|Grant%20Initial%20Charges
  (bind _handle_no_usage_expiration (Variables|Default|GetHandle%20No%20Usage%20Expiration))
  ;[1 1 1] Grant Initial Charges
  (if (Utilities|Time|IsValidTimerHandle _handle_no_usage_expiration)
    (Utilities|Time|ClearandInvalidateTimerbyHandle _handle_no_usage_expiration)
    (else
      _))
  (Utilities|FlowControl|Sequence
    (:then_0
      (bind _remaining_charges (CallFunction|GetRemainingCharges))
      (Variables|Default|SetInitial%20Charges%20to%20Modify (- (ScalableFloat|GetValueasInteger (Variables|Default|GetRow_Initial_Dash_Charges_Water)) _remaining_charges))
      (if (> (Variables|Default|GetInitial%20Charges%20to%20Modify) 0)
        (CallFunction|TryModifyCharges (Variables|Default|GetInitial%20Charges%20to%20Modify))))))

(event Custom|Start%20Charge%20Timer
  ;[1 1 1] Charge Timer
  ;;[0.9844 0.9615 0.0609 0.5] TODO: On elim, not timer
  (if (CallFunction|HaveMedallion?)
    _))

(event Custom|Recharge%20Dash
  ;[1 1 1] Charge Timer
  (if true
    (CallFunction|TryModifyCharges (ScalableFloat|GetValueasInteger (Variables|Default|GetRow_Dashes_Granted_on_Recharge)))
    (else
      (Utilities|Time|ClearandInvalidateTimerbyHandle (Variables|Default|GetTimer%20Handle%20Recharge%20Time))
      (Ability|HasAuthority
        (:True
          (bind _remaining_charges (CallFunction|GetRemainingCharges))
          (if (<= _remaining_charges 0)
            (Ability|RemoveGrantedbyEffect)
            (Ability|EndAbility)))) ; Calling on the client could cause the server to receive ending the ability from the client before this timer completes on the server
    )))

(event Custom|Use%20Dash%20Charge
  (params (Number%20of%20Charges :type Integer))
  ;[1 1 1] Use Dash Token
  (CallFunction|TryModifyCharges Number%20of%20Charges)
  ;;[1 0.25 0.25] Restart No Usage Timer
  (Utilities|Time|ClearandInvalidateTimerbyHandle (Variables|Default|GetHandle%20No%20Usage%20Expiration))
  (CallFunction|StartNoUsageExpirationTimer)
  (CallFunction|StartChargeTimer)
  (Ability|SendGameplayEvent "(TagName=\"Gameplay.Status.SunRose.SilkyWolf.Tutorial.End\")"))

(event Custom|Cleanup
  ;[1 1 1] Cleanup Effect
  (Ability|RemoveGameplayEffectFromOwnerWithGrantedTags "(GameplayTags=((TagName=\"Granted.Athena.SunRose.SilkyWolf.DashCharge\")))")
  (if (not (GameplayTags|HasTag (GameplayTags|GetOwnedGameplayTags (Variables|Components|GetPlayer%20Pawn) :!class /Script/GameplayTags.GameplayTagAssetInterface) "(TagName=\"Gameplay.Status.SunRose.SilkyWolf.Medallion\")"))
    (Ability|RemoveGrantedbyEffect)
    (Ability|EndAbility)))

(event Ability|EventOnEndAbility
  (params (bWasCancelled :type Boolean))
  (bind _handle_no_usage_expiration (Variables|Default|GetHandle%20No%20Usage%20Expiration))
  (bind _timer_handle_recharge_time (Variables|Default|GetTimer%20Handle%20Recharge%20Time))
  ;[1 1 1] Ability End and Clean Up
  (|CallParentFunction bWasCancelled)
  (GameplayTasks|EndTask (Variables|AsyncTasks|GetAsync%20Task%20Cooldown))
  (if (Utilities|Time|IsValidTimerHandle _handle_no_usage_expiration)
    (Utilities|Time|ClearandInvalidateTimerbyHandle _handle_no_usage_expiration))
  (if (Utilities|Time|IsValidTimerHandle _timer_handle_recharge_time)
    (Utilities|Time|ClearandInvalidateTimerbyHandle _timer_handle_recharge_time)))

(event Custom|Start%20No%20Usage%20Expiration%20Timer
  (bind _row_no_usage_expiration_timer (Variables|Default|GetRow%20No%20usage%20Expiration%20Timer))
  (bind _returnvalue (ScalableFloat|GetValueatLevel _row_no_usage_expiration_timer))
  ;[1 0.5208 0.5208] No Usage Expiration
  (if (> _returnvalue 0)
    (bind _returnvalue_2 (Utilities|Time|SetTimerbyEvent (delegate Custom|Set%20No%20Usage%20Expiration%20Timer) _returnvalue true))
    (Variables|Default|SetHandle%20No%20Usage%20Expiration _returnvalue_2)))

(event Custom|Set%20No%20Usage%20Expiration%20Timer
  ;[1 0.5208 0.5208] No Usage Expiration
  (bind _result (Math|Integer|NegateInt (ScalableFloat|GetValueasInteger (Variables|Default|GetRow%20No%20Usage%20Expiration%20Value)))
    (:%20%20
      (CallFunction|TryModifyCharges _result))))

(event Custom|Start%20AI%20Bot%20Dash%20Removal
  ;[1 0.5208 0.5208] AI Bot Removal Timer
  (if (GameplayTags|HasTag (GameplayTags|GetOwnedGameplayTags (Variables|Components|GetPlayer%20Pawn) :!class /Script/GameplayTags.GameplayTagAssetInterface) (Variables|Default|GetTag_Is_AI_Bot))
    (Utilities|Time|SetTimerbyEvent (delegate Custom|Start%20AI%20Bot%20Dash%20Removal%20Timer) (ScalableFloat|GetValueatLevel (Variables|Default|GetRow%20AI%20Bot%20Removal%20Timer)) true)))

(event Custom|Start%20AI%20Bot%20Dash%20Removal%20Timer
  ;[1 0.5208 0.5208] AI Bot Removal Timer
  (bind _result (Math|Integer|NegateInt (ScalableFloat|GetValueasInteger (Variables|Default|GetRow%20AI%20Bot%20Removal%20Value)))
    (:%20%20
      (CallFunction|UseDashCharge _result))))

(fn Chargecountchanged
  (Ability|HasAuthority
    (:True
      (bind _remaining_charges (CallFunction|GetRemainingCharges))
      (LocalVariables|SetRemaining%20Charge%20Count _remaining_charges)
      (if (== (LocalVariables|GetRemaining%20Charge%20Count) (Variables|Default|GetMax_Dash_Charges))
        (Utilities|Time|ClearandInvalidateTimerbyHandle (Variables|Default|GetTimer%20Handle%20Recharge%20Time))
        (else
          _))
      (if (> (LocalVariables|GetRemaining%20Charge%20Count) 0)
        _
        (elif (GameplayTags|HasTag (GameplayTags|GetOwnedGameplayTags (Variables|Components|GetPlayer%20Pawn) :!class /Script/GameplayTags.GameplayTagAssetInterface) "(TagName=\"Gameplay.Status.SunRose.SilkyWolf.Medallion\")")
          _
          (else
            (CallFunction|Cleanup))))
      (Ability|ApplyGameplayEffectToOwner (select (GameplayTags|HasMatchingGameplayTag (Variables|Components|GetPlayer%20Pawn) "(TagName=\"Granted.Athena.StyxDash.UseAlternateEffects\")") (Variables|Default|GetAlternateGameplayEffectClass) (Variables|Default|GetGameplay%20Effect%20Class))))))

(fn GetRemainingCharges
  (bind (_bsuccessfullyfoundattribute _returnvalue) (Ability|Attribute|GetFloatAttributefromAbilitySystemComponent (Variables|Default|GetAbility%20System%20Component) "(AttributeName=\"CurrentCharges\",Attribute=/Script/SunRoseEnvGameplayRuntime.FortChargingSet_SunroseDash:CurrentCharges,AttributeOwner=\"/Script/CoreUObject.Class'/Script/SunRoseEnvGameplayRuntime.FortChargingSet_SunroseDash'\")"))
  (return (Math|Float|Truncate _returnvalue)))

(fn GetDashRechargeTime
  (return (ScalableFloat|GetValueatLevel (Variables|Default|GetRow_Dash_Recharge_Time))))

(fn HaveMedallion?
  (return (GameplayTags|HasTag (GameplayTags|GetOwnedGameplayTags (Variables|Components|GetPlayer%20Pawn) :!class /Script/GameplayTags.GameplayTagAssetInterface) (Variables|Default|GetTag_SilkyWolf_Medallion))))

(fn InWater?
  (return (GameplayTags|HasTag (GameplayTags|GetOwnedGameplayTags (Variables|Components|GetPlayer%20Pawn) :!class /Script/GameplayTags.GameplayTagAssetInterface) (Variables|Default|GetTag_SilkyWolf_InWater))))

(fn TryModifyCharges
  (params (Number%20of%20Charges%20To%20Modify :type Integer))
  (bind _ability_system_component (Variables|Default|GetAbility%20System%20Component))
  (Ability|HasAuthority
    (:True
      (if (not (GameplayTags|HasMatchingGameplayTag (Variables|Components|GetPlayer%20Pawn) "(TagName=\"Granted.Athena.SunRose.SilkyWolf.Dash.BlockActivation\")"))
        (Ability|ApplyGameplayEffectToOwner "/SpaceClique/Gameplay/InAirDash/GE/GE_SpaceClique_DashChargeConsumeCooldown.GE_SpaceClique_DashChargeConsumeCooldown_C")
        (bind _remaining_charges (CallFunction|GetRemainingCharges))
        (bind _returnvalue (Ability|GameplayEffect|AssignTagSetbyCallerMagnitude (GameplayEffects|MakeOutgoingSpec :self _ability_system_component :GameplayEffectClass "/SpaceClique/Gameplay/InAirDash/GE/GE_SpaceClique_OnDashActivated.GE_SpaceClique_OnDashActivated_C" :Level 0.0) "(TagName=\"SetByCaller.DashCharges\")" (Math|Conversions|ToFloat(Integer) (Math|Integer|Min(Integer) (Variables|GetNumber%20of%20Charges%20To%20Modify) (- (Variables|Default|GetMax_Dash_Charges) _remaining_charges)))))
        (GameplayEffects|ApplyGameplayEffectSpecToSelf _ability_system_component _returnvalue)
        (CallFunction|ChargeCountChanged)
        (if (not (> (Variables|GetNumber%20of%20Charges%20To%20Modify) 0))
          (CallFunction|StartChargeTimer))
        (return)))))
